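# Physics asset for the lower straps on the F_MED_Geode female player body, giving the dynamic strap bones collision and joint limits.
# Capsules on both thighs and calves form the leg shapes, while boxes cover each left and right strap bone.
# Constraints chain the 02, 03, 04 and base bodies per side with tight swing and twist limits.
# The strap parent bodies and leg capsules have no constraints.
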PhysicsAsset /BRCosmetics/Characters/Player/Female/Medium/Bodies/F_MED_Geode/Meshes/F_MED_Geode_Bottom_Strap_Physics
bodies {
 calf_l [{shapeType=Capsule, length=26.6005, radius=7.9971, center=<-14.8402 -3.5222 0>, rotation=<rot 90 -0.0049 -0.0049>}]
 calf_r [{shapeType=Capsule, length=26.6005, radius=7.9971, center=<15.5774 3.0699 0.9455>, rotation=<rot 86.5266 -179.9999 179.9999>}]
 dyn_strap_02_bottom_l [{shapeType=Box, extentX=8.0131, extentY=18.8666, extentZ=3.0926, center=<1.8945 0 0.1764>}]
 dyn_strap_02_bottom_r [{shapeType=Box, extentX=8.0131, extentY=18.8666, extentZ=3.0926, center=<0.6663 0 0>}]
 dyn_strap_03_bottom_l1 [{shapeType=Box, extentX=10.0717, extentY=17.8441, extentZ=3.2308, center=<4.6591 0 -0.1163>}]
 dyn_strap_03_bottom_r [{shapeType=Box, extentX=8.2243, extentY=17.8441, extentZ=3.2308, center=<4.1063 0 0>}]
 dyn_strap_04_bottom_l [{shapeType=Box, extentX=10.0445, extentY=18.1217, extentZ=3.2465, center=<4.4591 1.5135 -0.528>}]
 dyn_strap_04_bottom_r [{shapeType=Box, extentX=10.0445, extentY=16.2232, extentZ=2.7084, center=<4.4591 0 0>}]
 dyn_strap_base_bottom_l [{shapeType=Box, extentX=3.734, extentY=3.734, extentZ=3.734}]
 dyn_strap_base_bottom_r [{shapeType=Box, extentX=3.734, extentY=3.734, extentZ=3.734}]
 dyn_strap_parent_bottom_l [{shapeType=Box, extentX=7.5, extentY=7.5, extentZ=7.5}]
 dyn_strap_parent_bottom_r [{shapeType=Box, extentX=7.5, extentY=7.5, extentZ=7.5}]
 thigh_l [{shapeType=Capsule, length=21.9372, radius=9.5479, center=<-25.3408 -0.0006 5.7906>, rotation=<rot 73.5912 -180 -180>}]
 thigh_r [{shapeType=Capsule, length=21.9372, radius=9.5479, center=<25.8354 -0.0002 -5.0484>, rotation=<rot 75.4265 -180 -180>}]
}
constraints [
 {bone1Name=dyn_strap_02_bottom_r, bone2Name=dyn_strap_03_bottom_r, swing1LimitDegrees=10, swing2LimitDegrees=10, twistLimitDegrees=3}
 {bone1Name=dyn_strap_03_bottom_r, bone2Name=dyn_strap_04_bottom_r, swing1LimitDegrees=5, swing2LimitDegrees=5, twistLimitDegrees=3}
 {bone1Name=dyn_strap_04_bottom_r, bone2Name=dyn_strap_base_bottom_r, swing1LimitDegrees=10, swing2LimitDegrees=10, twistLimitDegrees=3}
 {bone1Name=dyn_strap_04_bottom_l, bone2Name=dyn_strap_base_bottom_l, swing1LimitDegrees=10, swing2LimitDegrees=10, twistLimitDegrees=3}
 {bone1Name=dyn_strap_03_bottom_l1, bone2Name=dyn_strap_04_bottom_l, swing1LimitDegrees=8, swing2LimitDegrees=8, twistLimitDegrees=3}
 {bone1Name=dyn_strap_02_bottom_l, bone2Name=dyn_strap_03_bottom_l1, swing1LimitDegrees=8, swing2LimitDegrees=8, twistLimitDegrees=3}
]
