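# Physics asset for the jacket of the F_MED_Geode female cosmetic body, defining dynamic-bone physics for its left and right jacket pieces.
# Each side is a two-bone chain of box bodies constrained to spine_01, which carries two capsules approximating the torso.
# Joints have narrow twist and moderate swing limits.
# The sides differ slightly: the left lower joint locks Swing1 and the two lower boxes have different rotations.
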PhysicsAsset /BRCosmetics/Characters/Player/Female/Medium/Bodies/F_MED_Geode/Meshes/F_MED_Geode_Jacket_Physics
bodies {
 dyn_jacket_l_01 [{shapeType=Box, extentX=5.0916, extentY=6.6087, extentZ=2.3368, center=<1.9097 -0.0992 0.1687>, rotation=<rot -1.3641 -0.4851 -4.8148>}]
 dyn_jacket_l_02 [{shapeType=Box, extentX=5.6585, extentY=7.1805, extentZ=1.9966, center=<2.3042 0 0>, rotation=<rot 0 0 -6.3279>}]
 dyn_jacket_r_01 [{shapeType=Box, extentX=5.0916, extentY=6.6087, extentZ=2.3368, center=<1.9097 -0.0992 0.1687>, rotation=<rot -1.3641 -0.4851 -4.8148>}]
 dyn_jacket_r_02 [{shapeType=Box, extentX=5.6585, extentY=7.1805, extentZ=1.9966, center=<2.3042 0 0>, rotation=<rot 0 0 -3.703>}]
 spine_01 [
  {shapeType=Capsule, length=19.0497, radius=10.9503, center=<-4.4727 -0.1021 -3.2047>, rotation=<rot -79.8189 56.814 -26.694>}
  {shapeType=Capsule, length=19.0497, radius=10.9503, center=<-5.4768 0.4663 6.2684>, rotation=<rot -81.3917 -146.9091 167.6279>}
 ]
}
constraints [
 {bone1Name=dyn_jacket_l_02, bone2Name=dyn_jacket_l_01, swing1LimitDegrees=15, swing1Motion=Locked, swing2LimitDegrees=30, twistLimitDegrees=3}
 {bone1Name=dyn_jacket_r_01, bone2Name=spine_01, swing1LimitDegrees=20, swing2LimitDegrees=25, twistLimitDegrees=3}
 {bone1Name=dyn_jacket_l_01, bone2Name=spine_01, swing1LimitDegrees=20, swing2LimitDegrees=25, twistLimitDegrees=3}
 {bone1Name=dyn_jacket_r_02, bone2Name=dyn_jacket_r_01, swing1LimitDegrees=15, swing2LimitDegrees=30, twistLimitDegrees=3}
]
