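# Physics asset for the F_MED_Geode female player body, covering two dangling strap chains and the legs.
# Box bodies form left and right straps, chained parent, base, 03, 02 by angular constraints, with swing1 locked at the last joint.
# Capsules cover thighs and calf twists without constraints, apparently as strap colliders.
# Only the right strap's parent sphere is constrained to the pelvis, with free rotation; the left has no such constraint.
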PhysicsAsset /BRCosmetics/Characters/Player/Female/Medium/Bodies/F_MED_Geode/Meshes/F_MED_Geode_Physics
bodies {
 calf_twist_02_l [{shapeType=Capsule, length=45.4699, radius=6.5923, center=<2.3382 0.0002 -0.0619>, rotation=<rot 88.484 -0.0003 -0.0003>}]
 calf_twist_02_r [{shapeType=Capsule, length=45.4699, radius=6.5923, center=<-3.305 0 0.0875>, rotation=<rot 88.484 -0.0003 -0.0003>}]
 dyn_strap_02_top_l [{shapeType=Box, extentX=12.6177, extentY=17.3945, extentZ=3.7154, center=<4.317 0 2>, rotation=<rot 16.1805 0.0004 0>}]
 dyn_strap_02_top_r [{shapeType=Box, extentX=12.6177, extentY=17.3945, extentZ=3.9858, center=<4.9309 0 -0.6465>, rotation=<rot -16.1805 0 0>}]
 dyn_strap_03_top_l [{shapeType=Box, extentX=12.6536, extentY=21.0805, extentZ=5.1156, center=<4.992 0 2>, rotation=<rot 14.5187 0 0>}]
 dyn_strap_03_top_r [{shapeType=Box, extentX=12.6536, extentY=21.0805, extentZ=4.7334, center=<5.5414 0 -2>, rotation=<rot -14.5187 0 0>}]
 dyn_strap_base_top_l [{shapeType=Box, extentX=10.5279, extentY=22.5834, extentZ=2.3788, center=<4.4955 0 0>, rotation=<rot -4.149 0 0>}]
 dyn_strap_base_top_r [{shapeType=Box, extentX=10.5279, extentY=22.5834, extentZ=3.716, center=<4.4955 0.0001 -0.8256>, rotation=<rot -4.149 0 0>}]
 dyn_strap_parent_top_l [{shapeType=Sphere, radius=0.9706}]
 dyn_strap_parent_top_r [{shapeType=Sphere, radius=0.505}]
 pelvis [{shapeType=Box, extentX=2.293, extentY=2.293, extentZ=2.293}]
 thigh_l [
  {shapeType=Capsule, length=37.0455, radius=9.0975, center=<-18.1724 -1.3039 1.9614>, rotation=<rot -79.3336 0 0>}
  {shapeType=Capsule, length=11.9992, radius=6.4901, center=<-9.1384 -0.991 0.697>, rotation=<rot 78.6937 180 180>}
 ]
 thigh_r [
  {shapeType=Capsule, length=37.0455, radius=9.0975, center=<14.8621 -1.3035 -2.5873>, rotation=<rot -79.0741 0 0>}
  {shapeType=Capsule, length=32.3485, radius=6.9901, center=<10.3783 -0.9907 -0.5014>, rotation=<rot 79.9687 180 180>}
 ]
}
constraints [
 {bone1Name=dyn_strap_03_top_r, bone2Name=dyn_strap_base_top_r, swing1LimitDegrees=45, swing2LimitDegrees=12, twistLimitDegrees=5}
 {
  bone1Name=dyn_strap_02_top_r
  bone2Name=dyn_strap_03_top_r
  swing1LimitDegrees=40
  swing1Motion=Locked
  swing2LimitDegrees=12
  twistLimitDegrees=6
 }
 {bone1Name=dyn_strap_base_top_r, bone2Name=dyn_strap_parent_top_r, swing1LimitDegrees=50, swing2LimitDegrees=12, twistLimitDegrees=15}
 {
  bone1Name=dyn_strap_parent_top_r
  bone2Name=pelvis
  swing1LimitDegrees=45
  swing1Motion=Free
  swing2LimitDegrees=45
  swing2Motion=Free
  twistLimitDegrees=45
  twistMotion=Free
 }
 {bone1Name=dyn_strap_base_top_l, bone2Name=dyn_strap_parent_top_l, swing1LimitDegrees=50, swing2LimitDegrees=30, twistLimitDegrees=15}
 {bone1Name=dyn_strap_03_top_l, bone2Name=dyn_strap_base_top_l, swing1LimitDegrees=45, swing2LimitDegrees=25, twistLimitDegrees=5}
 {
  bone1Name=dyn_strap_02_top_l
  bone2Name=dyn_strap_03_top_l
  swing1LimitDegrees=40
  swing1Motion=Locked
  swing2LimitDegrees=20
  twistLimitDegrees=6
 }
]
