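# Post-body Animation Blueprint for the F_MED_Geode female cosmetic body, adding secondary motion to straps and jacket atop the pose copied from its attached parent mesh.
# Additive bone offsets adjust top straps for crouch and normal states, three chained RigidBody nodes simulate separate physics assets, then further offsets shape the jacket.
# The update event sets crouch and skydive/parachute alphas, damping strap physics while airborne, and applies a delayed upward external force when jumping.
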AnimBlueprint /BRCosmetics/Characters/Player/Female/Medium/Bodies/F_MED_Geode/Meshes/F_MED_Geode_AnimBp
target_skeleton=&/F_MED_Geode_Skeleton
graphs [
 AnimGraph {nodes [
   n1 = Get Ext_Force() -> (Ext_Force:Vector)
   n2 = Get SkydiveAlpha() -> (SkydiveAlpha:double)
   n3 = NOR Boolean!(A:bool=Get bIsSkyDiving(), B:bool=Get bIsCrouching()) -> (ReturnValue:bool)
   n4 = Get CrouchAlpha() -> (CrouchAlpha:double)
   n5 = "Transform (Modify) Bone - Bone: dyn_strap_parent_top_r"(
    ComponentPose=Local To Component(LocalPose=Copy Pose From Mesh() {bUseAttachedParent=true})
    Alpha=@n4
    Translation="0.000,3,6"
    Rotation="0,0,0"
    Scale="1,1,1"
   ) {
    rotationMode=BMM_Additive
    rotationSpace=BCS_BoneSpace
    translationMode=BMM_Additive
    translationSpace=BCS_BoneSpace
    alphaBoolBlend {blendOption=Linear}
    boneToModify {boneName=dyn_strap_parent_top_r}
    translation {y=3, z=6}
   }
   n6 = "Transform (Modify) Bone - Bone: dyn_strap_parent_top_l"(ComponentPose=@n5, Alpha=@n4, Translation="0.000,-3,5", Rotation="20,0,0", Scale="1,1,1") {
    rotationMode=BMM_Additive
    rotationSpace=BCS_BoneSpace
    translationMode=BMM_Additive
    translationSpace=BCS_BoneSpace
    alphaBoolBlend {blendOption=Linear}
    boneToModify {boneName=dyn_strap_parent_top_l}
    rotation {pitch=20}
    translation {y=-3, z=5}
   }
   n7 = "Transform (Modify) Bone - Bone: dyn_strap_parent_top_l"!(ComponentPose=@n6, bAlphaBoolEnabled:bool=@n3, Translation="0,0,0", Rotation="50,0,0", Scale="1,1,1") -> (Pose) {
    alphaInputType=Bool
    rotationMode=BMM_Additive
    rotationSpace=BCS_BoneSpace
    scaleMode=BMM_Additive
    scaleSpace=BCS_BoneSpace
    translationMode=BMM_Additive
    translationSpace=BCS_BoneSpace
    alphaBoolBlend {blendOption=Linear}
    boneToModify {boneName=dyn_strap_parent_top_l}
    rotation {pitch=50}
   }
   n8 = "Transform (Modify) Bone - Bone: dyn_strap_parent_top_r"!(ComponentPose=@n7, bAlphaBoolEnabled:bool=@n3, Translation="0,0,0", Rotation="25,0,0", Scale="1,1,1") -> (Pose) {
    alphaInputType=Bool
    rotationMode=BMM_Additive
    rotationSpace=BCS_BoneSpace
    scaleMode=BMM_Additive
    scaleSpace=BCS_BoneSpace
    translationMode=BMM_Additive
    translationSpace=BCS_BoneSpace
    alphaBoolBlend {blendOption=Linear}
    boneToModify {boneName=dyn_strap_parent_top_r}
    rotation {pitch=25}
   }
   n9 = RigidBody(ComponentPose=@n8, Alpha=@n2, ExternalForce=@n1) {
    bClampLinearTranslationLimitToRefPose=true
    bOverrideWorldGravity=true
    overridePhysicsAsset=&/F_MED_Geode_Physics
    simulationSpace=BaseBoneSpace
    alphaBoolBlend {blendOption=Linear}
    baseBoneRef {boneName=root}
    componentAppliedLinearAccClamp=<9000 9000 9000>
    componentLinearAccScale=<0.25 0.25 0.25>
    componentLinearVelScale=<0.25 0.25 0.25>
    overrideWorldGravity {z=-980}
   }
   n10 = RigidBody(ComponentPose=@n9, Alpha=@n2, ExternalForce=@n1) {
    bClampLinearTranslationLimitToRefPose=true
    bOverrideWorldGravity=true
    overridePhysicsAsset=&/F_MED_Geode_Bottom_Strap_Physics
    simulationSpace=BaseBoneSpace
    alphaBoolBlend {blendOption=Linear}
    baseBoneRef {boneName=root}
    componentAppliedLinearAccClamp=<5000 5000 5000>
    componentLinearAccScale=<0.25 0.25 0.25>
    componentLinearVelScale=<0.25 0.25 0.25>
    overrideWorldGravity {z=-980}
   }
   n11 = RigidBody(ComponentPose=@n10, Alpha=0.85, ExternalForce=Get Ext_Force()) {
    bClampLinearTranslationLimitToRefPose=true
    bOverrideWorldGravity=true
    overridePhysicsAsset=&/F_MED_Geode_Jacket_Physics
    simulationSpace=BaseBoneSpace
    alphaBoolBlend {blendOption=Linear}
    baseBoneRef {boneName=spine_01}
    componentAppliedLinearAccClamp=<5000 5000 5000>
    componentLinearAccScale=<0.25 0.25 0.25>
    componentLinearVelScale=<0.25 0.25 0.25>
    overrideWorldGravity {z=-980}
   }
   n12 = Get Skydive_Jacket() -> (Skydive_Jacket:double)
   n13 = "Transform (Modify) Bone - Bone: dyn_jacket_r_01"(ComponentPose=@n11, Alpha=@n12, Translation="0,0,0", Rotation="20,0,0", Scale="1,1,1") {
    rotationMode=BMM_Additive
    rotationSpace=BCS_BoneSpace
    alphaBoolBlend {blendOption=Linear}
    boneToModify {boneName=dyn_jacket_r_01}
    rotation {pitch=20}
   }
   n14 = "Transform (Modify) Bone - Bone: dyn_jacket_l_01"(ComponentPose=@n13, Alpha=@n12, Translation="0,0,0", Rotation="20,0,0", Scale="1,1,1") {
    rotationMode=BMM_Additive
    rotationSpace=BCS_BoneSpace
    alphaBoolBlend {blendOption=Linear}
    boneToModify {boneName=dyn_jacket_l_01}
    rotation {pitch=20}
   }
   n15 = "Transform (Modify) Bone - Bone: dyn_jacket_l_01"(ComponentPose=@n14, Alpha=1, Translation="0,0,0", Rotation="7,3.5,0", Scale="1,1,1") {
    alpha=0
    rotationMode=BMM_Additive
    rotationSpace=BCS_BoneSpace
    alphaBoolBlend {blendOption=Linear}
    boneToModify {boneName=dyn_jacket_l_01}
    rotation {pitch=7, yaw=3.5}
   }
   n16 = "Transform (Modify) Bone - Bone: dyn_jacket_r_01"(ComponentPose=@n15, Alpha=1, Translation="0,0,0", Rotation="7,-3.5,0", Scale="1,1,1") {
    alpha=0
    rotationMode=BMM_Additive
    rotationSpace=BCS_BoneSpace
    alphaBoolBlend {blendOption=Linear}
    boneToModify {boneName=dyn_jacket_r_01}
    rotation {pitch=7, yaw=-3.5}
   }
   Output Pose(Result=Component To Local(ComponentPose=@n16)) {layerGroup=None}
  ]}
 EventGraph {nodes [
   Event BlueprintUpdateAnimation()
   > n1 = Sequence!(execute) -> (then_0, then_1, then_2:Exec, then_3:Exec)
   > Set CrouchAlpha(CrouchAlpha:double="To Float (Boolean)"(InBool=Get bIsCrouching())) -> (Output_Get:double)
   n2 = Get bIsSkydiveDiveMode()
   n3 = Get bIsSkydiveFloating()
   n4 = Get bIsSkyDiving()
   n5 = Get bIsParachuteOpened()
   Set Skydive_Jacket(execute=@n1.then_3, Skydive_Jacket:double=SelectFloat(A=1, B=0, bPickA=OR Boolean!(A=@n4, B=@n5, C:bool=@n3, D:bool=@n2) -> (ReturnValue))) -> (Output_Get:double)
   n6 = Get bIsSkydiveDiveMode()
   n7 = Get bIsSkydiveFloating()
   n8 = Get bIsSkyDiving()
   n9 = Get bIsParachuteOpened()
   Set SkydiveAlpha(execute=@n1.then_1, SkydiveAlpha:double=SelectFloat(A=0.25, B=1, bPickA=OR Boolean(A=@n8, B=@n9, C=@n7, D=@n6))) -> (Output_Get:double)
   n10 = Branch(execute=@n1.then_2, Condition=Get bIsJumping())
   > Delay(Duration=0.25)
   > Set Ext_Force(Ext_Force:Vector="0, 0,1250") -> (Output_Get:Vector)
   > n11 = Delay(Duration=0.375)
   Do Once(execute=@n10.else, Reset=@n11.then)
   > Set Ext_Force(Ext_Force="0, 0, 0")
   TryGetPawnOwner()
  ]}
]
variables [
 CrouchAlpha {type=number}
 Ext_Force {type=Vector}
 Skydive_Jacket {type=number}
 SkydiveAlpha {type=number}
]
