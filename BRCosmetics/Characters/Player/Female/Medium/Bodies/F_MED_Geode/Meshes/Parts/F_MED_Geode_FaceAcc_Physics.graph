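# Physics asset for the F_MED_Geode face accessory mesh, defining simulation bodies for the character's dynamic bangs and hood.
# Each of the six bangs bones, three hood bones and the head has a single box body.
# All nine dynamic bones are constrained directly to the head rather than chained, with tight twist limits.
# Swing limits vary per bone, with the central hood bone far looser than the bangs and hood sides.
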PhysicsAsset /BRCosmetics/Characters/Player/Female/Medium/Bodies/F_MED_Geode/Meshes/Parts/F_MED_Geode_FaceAcc_Physics
bodies {
 dyn_bangs_01 [{shapeType=Box, extentX=3.6033, extentY=1.4641, extentZ=2.2098, center=<1.0881 0 0>}]
 dyn_bangs_02 [{shapeType=Box, extentX=2.9955, extentY=2.2657, extentZ=1.8047, center=<0.8938 0 0>}]
 dyn_bangs_03 [{shapeType=Box, extentX=2.4848, extentY=1.4498, extentZ=1.3314, center=<0.6746 0 0>}]
 dyn_bangs_04 [{shapeType=Box, extentX=2.9427, extentY=1.9135, extentZ=1.7416, center=<1.0346 -0.2357 -0.2101>}]
 dyn_bangs_05 [{shapeType=Box, extentX=2.9985, extentY=1.6087, extentZ=3.9985, center=<1.0779 -0.0209 -0.0613>}]
 dyn_bangs_06 [{shapeType=Box, extentX=3.038, extentY=1.7687, extentZ=1.7344, center=<0.5309 0 0>}]
 dyn_hood_01 [{shapeType=Box, extentX=7.1132, extentY=9.0523, extentZ=1.7469}]
 dyn_hood_l [{shapeType=Box, extentX=5.0788, extentY=6.0753, extentZ=2.3003}]
 dyn_hood_r [{shapeType=Box, extentX=5.0788, extentY=6.0753, extentZ=2.3003}]
 head [{shapeType=Box, extentX=2.6184, extentY=0.5, extentZ=2.8985}]
}
constraints [
 {bone1Name=dyn_bangs_05, bone2Name=head, swing1LimitDegrees=25, swing2LimitDegrees=40, twistLimitDegrees=15}
 {bone1Name=dyn_bangs_04, bone2Name=head, swing1LimitDegrees=20, swing2LimitDegrees=20, twistLimitDegrees=5}
 {bone1Name=dyn_bangs_03, bone2Name=head, swing1LimitDegrees=35, swing2LimitDegrees=15, twistLimitDegrees=5}
 {bone1Name=dyn_bangs_02, bone2Name=head, swing1LimitDegrees=20, swing2LimitDegrees=30, twistLimitDegrees=5}
 {bone1Name=dyn_bangs_01, bone2Name=head, swing1LimitDegrees=15, swing2LimitDegrees=30, twistLimitDegrees=5}
 {bone1Name=dyn_bangs_06, bone2Name=head, swing1LimitDegrees=25, swing2LimitDegrees=40, twistLimitDegrees=3}
 {bone1Name=dyn_hood_01, bone2Name=head, swing1LimitDegrees=70, swing2LimitDegrees=70, twistLimitDegrees=3}
 {bone1Name=dyn_hood_r, bone2Name=head, swing1LimitDegrees=12, swing2LimitDegrees=15, twistLimitDegrees=3}
 {bone1Name=dyn_hood_l, bone2Name=head, swing1LimitDegrees=12, swing2LimitDegrees=15, twistLimitDegrees=3}
]
