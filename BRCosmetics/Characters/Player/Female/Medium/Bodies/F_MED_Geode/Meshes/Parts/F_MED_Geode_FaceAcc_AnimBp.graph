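# Animation Blueprint for the face accessory of the female F_MED_Geode character body, adding secondary physics motion.
# It copies the pose from the attached parent mesh, then runs a partially blended RigidBody simulation in spine_02 bone space with its own physics asset.
# The event graph checks bIsJumping each update, applying a delayed upward external force during jumps and zeroing it through a resettable Do Once otherwise.
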AnimBlueprint /BRCosmetics/Characters/Player/Female/Medium/Bodies/F_MED_Geode/Meshes/Parts/F_MED_Geode_FaceAcc_AnimBp
target_skeleton=&/F_MED_Geode_FaceAcc_Skeleton
graphs [
 AnimGraph {nodes [
   n1 = RigidBody(
    ComponentPose=Local To Component(LocalPose=Copy Pose From Mesh() {bUseAttachedParent=true})
    Alpha=0.75
    ExternalForce=Get Ext_Force() -> (Ext_Force:Vector)
   ) {
    bClampLinearTranslationLimitToRefPose=true
    bOverrideWorldGravity=true
    overridePhysicsAsset=&/F_MED_Geode_FaceAcc_Physics
    simulationSpace=BaseBoneSpace
    alphaBoolBlend {blendOption=Linear}
    baseBoneRef {boneName=spine_02}
    componentAppliedLinearAccClamp=<6000 6000 6000>
    componentLinearAccScale=<0.75 0.75 0.75>
    componentLinearVelScale=<0.75 0.75 0.75>
    overrideWorldGravity {z=-980}
   }
   Output Pose(Result=Component To Local(ComponentPose=@n1)) {layerGroup=None}
  ]}
 EventGraph {nodes [
   Event BlueprintUpdateAnimation()
   > n1 = Branch(Condition=Get bIsJumping())
   > Delay(Duration=0.25)
   > Set Ext_Force(Ext_Force:Vector="0, 0,1300") -> (Output_Get:Vector)
   > n2 = Delay(Duration=0.37)
   Do Once(execute=@n1.else, Reset=@n2.then)
   > Set Ext_Force(Ext_Force="0, 0, 0")
   TryGetPawnOwner()
  ]}
]
variables [Ext_Force {type=Vector}]
